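# Parent character material for the PineTrimPack Sift body skin, adding a timed, reactive glowing-symbol effect between the shared 2023 PreFX and PostFX functions.
# Time since the BP_GetTime timestamp drives an envelope that fades in, then eases out after EffectDuration, scaled by PineTrimPack_UseReactivity.
# That envelope gates boosted, power-curved emissive and darkens base color where Tech Art Mask blue minus red applies.
# The huge default timestamp keeps the effect off until overridden.
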Material /BRCosmetics/Characters/Player/Male/Medium/Bodies/M_MED_PineTrimPack_Sift/Materials/M_PineTrimPack_Sift_Parent
graph {
 properties {
  bUseMaterialAttributes=true
  bUsedWithMeshDeformer=true
  bUsedWithMorphTargets=true
  bUsedWithSkeletalMesh=true
  pixelDepthOffsetMode=PDOM_Legacy
  shadingModel=MSM_FromMaterialExpression
  usedShadingModels=Default Lit | Subsurface | Clear Coat | Subsurface Profile | Cloth
 }
 nodes [
  n1 = MaterialFunctionCall() {materialFunction=&MF_CharacterParent_2023_PreFX}
  n2 = GetMaterialAttributes(None=@n1.Result) {attributeGetTypes [#1, #2]}
  n3 = Multiply(A=Power(Base=@n2.Emissive Color) {constExponent=1.75}) {constB=7.15}
  n4 = NamedRerouteUsage()
  n5 = Multiply(A=ComponentMask(None=@n1.Tech Art Mask) {b=true}) {constB=3}
  n6 = Multiply(A=ComponentMask(None=@n1.Tech Art Mask) {r=true}) {constB=3}
  n7 = Multiply(A=@n4, B=Subtract(A=Saturate(None=@n5), B=Saturate(None=@n6)))
  MaterialFunctionCall(
   Material=SetMaterialAttributes!(
    MaterialAttributes=@n2.MaterialAttributes
    Emissive Color=Multiply(A=@n3, B=@n4)
    Base Color=LinearInterpolate(A=@n2.Base Color, Alpha=@n7) {constB=0.05}
   ) {attributeSetTypes [#1, #2]}
   M=@n1.M
   Pre Skinned Local Position=@n1.M
   Distance Blend=@n1.M
   Fresnel=@n1.M
  ) {materialFunction=&MF_CharacterParent_2023_PostFX}
  n8 = Subtract(A=Time(), B=ScalarParameter() {defaultValue=9.999999848243207e+30, group=PineTrimPack, parameterName=BP_GetTime, sortPriority=100})
  n9 = Subtract(A=@n8, B=ScalarParameter() {defaultValue=6, parameterName=EffectDuration}) {constB=5}
  n10 = OneMinus(None=Saturate(None=Multiply(A=@n9) {constB=0.7}))
  n11 = Multiply(A=@n10, B=@n10)
  n12 = Saturate(None=Multiply(A=Subtract(A=@n8) {constB=0.3}) {constB=1.5})
  NamedRerouteDeclaration(None=Multiply(A=Multiply(A=@n11, B=@n12), B=ScalarParameter() {defaultValue=1, group=PineTrimPack, parameterName=PineTrimPack_UseReactivity})) {name=TimeStamp_GlowingSymbols, nodeColor=<rgba 0 1 0.0625 1>}
 ]
}
