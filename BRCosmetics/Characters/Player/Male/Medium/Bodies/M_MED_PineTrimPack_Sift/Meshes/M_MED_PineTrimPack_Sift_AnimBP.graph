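# Post-process style animation Blueprint for the PineTrimPack Sift male body cosmetic, adding secondary motion to necklace, hood, zipper, baggy clothing and pin bones.
# The pose copied from the parent mesh runs through a Control Rig, additive bone offsets, then chained Anim Dynamics, cached between stages.
# Several bones are simulated twice, often mixing relative and world space.
# The event graph drives material WPO cloth wind, turbulence and velocity, plus parameter-asset gravity overrides.
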AnimBlueprint /BRCosmetics/Characters/Player/Male/Medium/Bodies/M_MED_PineTrimPack_Sift/Meshes/M_MED_PineTrimPack_Sift_AnimBP
target_skeleton=&/M_MED_PineTrimPack_Sift_Skeleton
graphs [
 AnimGraph {nodes [
   n1 = Anim Dynamics - Bone: dyn_necklees_01!(
    ComponentPose=Local To Component(LocalPose=Use cached pose 'AdjustCache'())
    Alpha=1
    GravityOverride:Vector="0,0,-980"
    ExternalForce="0,0,0"
   ) -> (Pose) {
    angularDampingOverride=1
    angularSpringConstant=4500
    bAngularSpring=true
    bOverrideAngularDamping=true
    bUseGravityOverride=true
    gravityScale=0.35
    linearDampingOverride=1
    linearSpringConstant=120
    simulationSpace=RootRelative
    alphaBoolBlend {blendOption=Linear}
    boundBone {boneName=dyn_necklees_01}
    componentAppliedLinearAccClamp=<3500 3500 3500>
    componentLinearAccScale=<0.3 0.3 0.3>
    componentLinearVelScale=<0.3 0.3 0.3>
    gravityOverride {z=-980}
    physicsBodyDefinitions [{
      boundBone {boneName=dyn_necklees_01}
      boxExtents=<20 20 20>
      constraintSetup {angularLimitsMax=<0 0 5>, angularLimitsMin=<0 -20 -5>, angularTarget=<1 0 0>}
      localJointOffset=<2 0 0>
     }]
    relativeSpaceBone {boneName=root}
    simSpaceSettings {maxAngularAcceleration=100, maxAngularVelocity=10}
   }
   n2 = Anim Dynamics - Bone: dyn_necklees_02!(ComponentPose=@n1, Alpha=0.5, GravityOverride:Vector="0,0,-980", ExternalForce="0,0,0") -> (Pose) {
    angularDampingOverride=1
    angularSpringConstant=4500
    bAngularSpring=true
    bOverrideAngularDamping=true
    bUseGravityOverride=true
    gravityScale=0.35
    linearDampingOverride=1
    linearSpringConstant=120
    simulationSpace=RootRelative
    alphaBoolBlend {blendOption=Linear}
    boundBone {boneName=dyn_necklees_02}
    componentAppliedLinearAccClamp=<3500 3500 3500>
    componentLinearAccScale=<0.3 0.3 0.3>
    componentLinearVelScale=<0.3 0.3 0.3>
    gravityOverride {z=-980}
    physicsBodyDefinitions [{
      boundBone {boneName=dyn_necklees_02}
      boxExtents=<20 20 20>
      constraintSetup {angularLimitsMax=<0 -7 0>, angularLimitsMin=<0 -50 0>, angularTarget=<1 0 0>}
      localJointOffset=<2 0 0>
     }]
    simSpaceSettings {maxAngularAcceleration=100, maxAngularVelocity=10}
   }
   n3 = Anim Dynamics - Bone: dyn_necklees_01(ComponentPose=@n2, Alpha=0.3, GravityOverride="0,0,-980", ExternalForce="0,0,0") {
    angularDampingOverride=1
    angularSpringConstant=4500
    bAngularSpring=true
    bOverrideAngularDamping=true
    bUseGravityOverride=true
    gravityScale=0.35
    linearDampingOverride=1
    linearSpringConstant=120
    simulationSpace=World
    alphaBoolBlend {blendOption=Linear}
    boundBone {boneName=dyn_necklees_01}
    componentAppliedLinearAccClamp=<3500 3500 3500>
    componentLinearAccScale=<0.3 0.3 0.3>
    componentLinearVelScale=<0.3 0.3 0.3>
    gravityOverride {z=-980}
    physicsBodyDefinitions [{
      boundBone {boneName=dyn_necklees_01}
      boxExtents=<20 20 20>
      constraintSetup {angularLimitsMax=<0 0 7>, angularLimitsMin=<0 -20 -7>, angularTarget=<1 0 0>}
      localJointOffset=<2 0 0>
     }]
    simSpaceSettings {maxAngularAcceleration=100, maxAngularVelocity=10}
   }
   n4 = Anim Dynamics - Bone: dyn_necklees_02(ComponentPose=@n3, Alpha=0.3, GravityOverride="0,0,-980", ExternalForce="0,0,0") {
    angularDampingOverride=1
    angularSpringConstant=4500
    bAngularSpring=true
    bOverrideAngularDamping=true
    bUseGravityOverride=true
    gravityScale=0.35
    linearDampingOverride=1
    linearSpringConstant=120
    simulationSpace=World
    alphaBoolBlend {blendOption=Linear}
    boundBone {boneName=dyn_necklees_02}
    componentAppliedLinearAccClamp=<3500 3500 3500>
    componentLinearAccScale=<1 1 1>
    componentLinearVelScale=<1 1 1>
    gravityOverride {z=-980}
    physicsBodyDefinitions [{
      boundBone {boneName=dyn_necklees_02}
      boxExtents=<25 25 25>
      constraintSetup {angularLimitsMax=<0 -2 30>, angularLimitsMin=<0 10 -30>, angularTarget=<1 0 0>}
      localJointOffset=<2 0 0>
     }]
    simSpaceSettings {maxAngularAcceleration=100, maxAngularVelocity=10}
   }
   n5 = Anim Dynamics - Bone: dyn_hood_dn_01!(ComponentPose=@n4, Alpha=0.7, GravityOverride:Vector="0,0,-980", ExternalForce="0,0,0") -> (Pose) {
    angularDampingOverride=1
    angularSpringConstant=15000
    bAngularSpring=true
    bOverrideAngularDamping=true
    bUseGravityOverride=true
    gravityScale=0.35
    linearDampingOverride=0.9
    linearSpringConstant=20
    simulationSpace=RootRelative
    alphaBoolBlend {blendOption=Linear}
    boundBone {boneName=dyn_hood_dn_01}
    componentAppliedLinearAccClamp=<3000 3000 3000>
    componentLinearAccScale=<0.7 1 0.7>
    componentLinearVelScale=<0.7 1 0.7>
    gravityOverride {z=-980}
    physicsBodyDefinitions [{
      boundBone {boneName=dyn_hood_dn_01}
      boxExtents=<30 50 30>
      constraintSetup {angularLimitsMax=<5 0 15>, angularLimitsMin=<-5 -40 -15>, angularTarget=<1 0 0>, linearAxesMax=<0 0 1>}
      localJointOffset=<8 0 0>
     }]
    simSpaceSettings {maxAngularAcceleration=150, maxAngularVelocity=50, simSpaceAngularAlpha=0.2, externalAngularVelocity=<6 6 6>}
   }
   n6 = Anim Dynamics - Bone: dyn_hood_dn_02!(ComponentPose=@n5, Alpha=0.7, GravityOverride:Vector="0,0,-980", ExternalForce="0,0,0") -> (Pose) {
    angularDampingOverride=1
    angularSpringConstant=15000
    bAngularSpring=true
    bOverrideAngularDamping=true
    bUseGravityOverride=true
    gravityScale=0.35
    linearDampingOverride=0.9
    linearSpringConstant=20
    simulationSpace=RootRelative
    alphaBoolBlend {blendOption=Linear}
    boundBone {boneName=dyn_hood_dn_02}
    componentAppliedLinearAccClamp=<3000 3000 3000>
    componentLinearAccScale=<0.7 1 0.7>
    componentLinearVelScale=<0.7 1 0.7>
    gravityOverride {z=-980}
    physicsBodyDefinitions [{
      boundBone {boneName=dyn_hood_dn_02}
      boxExtents=<30 50 30>
      constraintSetup {angularLimitsMax=<5 0 15>, angularLimitsMin=<-5 -40 -15>, angularTarget=<1 0 0>, linearAxesMax=<0 0 1>}
      localJointOffset=<8 0 0>
     }]
    simSpaceSettings {maxAngularAcceleration=150, maxAngularVelocity=50, simSpaceAngularAlpha=0.2, externalAngularVelocity=<6 6 6>}
   }
   n7 = Anim Dynamics - Bone: dyn_zip_02!(ComponentPose=@n6, Alpha=1, GravityOverride:Vector="0,0,-980", ExternalForce="0,0,0") -> (Pose) {
    angularDampingOverride=0.9999
    angularSpringConstant=25000
    bAngularSpring=true
    bOverrideAngularDamping=true
    bUseGravityOverride=true
    gravityScale=0.35
    linearDampingOverride=1
    linearSpringConstant=120
    simulationSpace=RootRelative
    alphaBoolBlend {blendOption=Linear}
    boundBone {boneName=dyn_zip_02}
    componentAppliedLinearAccClamp=<3500 3500 3500>
    componentLinearAccScale=<1 1 1>
    componentLinearVelScale=<1 1 1>
    gravityOverride {z=-980}
    physicsBodyDefinitions [{
      boundBone {boneName=dyn_zip_02}
      boxExtents=<20 20 20>
      constraintSetup {angularLimitsMax=<30 60 30>, angularLimitsMin=<-30 -60 -30>, angularTarget=<1 0 0>}
      localJointOffset=<10 0 0>
     }]
    simSpaceSettings {maxAngularAcceleration=1000, maxAngularVelocity=100, simSpaceAngularAlpha=0.2}
   }
   n8 = Anim Dynamics - Bone: dyn_zip_02(ComponentPose=@n7, Alpha=1, GravityOverride="0,0,-980", ExternalForce="0,0,0") {
    angularDampingOverride=0.9999
    angularSpringConstant=25000
    bAngularSpring=true
    bOverrideAngularDamping=true
    bUseGravityOverride=true
    gravityScale=0.35
    linearDampingOverride=1
    linearSpringConstant=120
    simulationSpace=RootRelative
    alphaBoolBlend {blendOption=Linear}
    boundBone {boneName=dyn_zip_02}
    componentAppliedLinearAccClamp=<3500 3500 3500>
    componentLinearAccScale=<1 1 1>
    componentLinearVelScale=<1 1 1>
    gravityOverride {z=-980}
    physicsBodyDefinitions [{
      boundBone {boneName=dyn_zip_02}
      boxExtents=<20 20 20>
      constraintSetup {angularLimitsMax=<3 0 30>, angularLimitsMin=<-3 -60 -30>, angularTarget=<1 0 0>}
      localJointOffset=<10 0 0>
     }]
    simSpaceSettings {maxAngularAcceleration=1000, maxAngularVelocity=100, simSpaceAngularAlpha=0.2}
   }
   n9 = Anim Dynamics - Bone: dyn_fr_bagginess_c!(ComponentPose=@n8, Alpha=1, GravityOverride:Vector="0,0,-980", ExternalForce="0,0,0") -> (Pose) {
    angularDampingOverride=0.999
    angularSpringConstant=8000
    bAngularSpring=true
    bLinearSpring=true
    bOverrideAngularDamping=true
    bOverrideLinearDamping=true
    bUseGravityOverride=true
    gravityScale=0.35
    linearDampingOverride=0.999
    linearSpringConstant=100
    simulationSpace=BoneRelative
    alphaBoolBlend {blendOption=Linear}
    boundBone {boneName=dyn_fr_bagginess_c}
    componentAppliedLinearAccClamp=<3000 3000 3000>
    componentLinearAccScale=<1 1 1>
    componentLinearVelScale=<1 1 1>
    gravityOverride {z=-980}
    physicsBodyDefinitions [{
      boundBone {boneName=dyn_fr_bagginess_c}
      boxExtents=<30 30 30>
      constraintSetup {angularLimitsMax=<0 30 10>, angularLimitsMin=<0 -30 -10>, angularTarget=<1 0 0>, linearAxesMax=<0 0 3>}
      localJointOffset=<7 0 0>
     }]
    relativeSpaceBone {boneName=pelvis}
   }
   n10 = Anim Dynamics - Bone: dyn_sd_bagginess_l!(ComponentPose=@n9, Alpha=1, GravityOverride:Vector="0,0,-980", ExternalForce="0,0,0") -> (Pose) {
    angularDampingOverride=1
    angularSpringConstant=7000
    bAngularSpring=true
    bLinearSpring=true
    bOverrideAngularDamping=true
    bOverrideLinearDamping=true
    bUseGravityOverride=true
    gravityScale=0.35
    linearDampingOverride=0.999
    linearSpringConstant=200
    simulationSpace=BoneRelative
    alphaBoolBlend {blendOption=Linear}
    boundBone {boneName=dyn_sd_bagginess_l}
    componentAppliedLinearAccClamp=<3000 3000 3000>
    componentLinearAccScale=<0.45 0.45 0.45>
    componentLinearVelScale=<0.45 0.45 0.45>
    gravityOverride {z=-980}
    physicsBodyDefinitions [{
      boundBone {boneName=dyn_sd_bagginess_l}
      boxExtents=<30 30 30>
      constraintSetup {angularLimitsMax=<10 30 10>, angularLimitsMin=<-10 -30 -10>, angularTarget=<1 0 0>, linearAxesMax=<0 0 3>}
      localJointOffset=<1.5 0 0>
     }]
    relativeSpaceBone {boneName=root}
   }
   n11 = Anim Dynamics - Bone: dyn_sd_bagginess_r!(ComponentPose=@n10, Alpha=1, GravityOverride:Vector="0,0,-980", ExternalForce="0,0,0") -> (Pose) {
    angularDampingOverride=1
    angularSpringConstant=7000
    bAngularSpring=true
    bLinearSpring=true
    bOverrideAngularDamping=true
    bOverrideLinearDamping=true
    bUseGravityOverride=true
    gravityScale=0.35
    linearDampingOverride=0.999
    linearSpringConstant=200
    simulationSpace=BoneRelative
    alphaBoolBlend {blendOption=Linear}
    boundBone {boneName=dyn_sd_bagginess_r}
    componentAppliedLinearAccClamp=<3000 3000 3000>
    componentLinearAccScale=<0.45 0.45 0.45>
    componentLinearVelScale=<0.45 0.45 0.45>
    gravityOverride {z=-980}
    physicsBodyDefinitions [{
      boundBone {boneName=dyn_sd_bagginess_r}
      boxExtents=<30 30 30>
      constraintSetup {angularLimitsMax=<10 30 10>, angularLimitsMin=<-10 -30 -10>, angularTarget=<1 0 0>, linearAxesMax=<0 0 3>}
      localJointOffset=<1.5 0 0>
     }]
    relativeSpaceBone {boneName=root}
   }
   n12 = Anim Dynamics - Bone: dyn_bk_bagginess_c!(ComponentPose=@n11, Alpha=1, GravityOverride:Vector="0,0,-980", ExternalForce="0,0,0") -> (Pose) {
    angularDampingOverride=1
    angularSpringConstant=7000
    bAngularSpring=true
    bLinearSpring=true
    bOverrideAngularDamping=true
    bOverrideLinearDamping=true
    bUseGravityOverride=true
    gravityScale=0.35
    linearDampingOverride=0.999
    linearSpringConstant=200
    simulationSpace=BoneRelative
    alphaBoolBlend {blendOption=Linear}
    boundBone {boneName=dyn_bk_bagginess_c}
    componentAppliedLinearAccClamp=<3000 3000 3000>
    componentLinearAccScale=<0.45 0.45 0.45>
    componentLinearVelScale=<0.45 0.45 0.45>
    gravityOverride {z=-980}
    physicsBodyDefinitions [{
      boundBone {boneName=dyn_bk_bagginess_c}
      boxExtents=<30 30 30>
      constraintSetup {angularLimitsMax=<10 30 10>, angularLimitsMin=<-10 -30 -10>, angularTarget=<1 0 0>, linearAxesMax=<0 0 3>}
      localJointOffset=<1.5 0 0>
     }]
    relativeSpaceBone {boneName=root}
   }
   n13 = Anim Dynamics - Bone: dyn_fr_bagginess_c(ComponentPose=@n12, Alpha=1, GravityOverride="0,0,-980", ExternalForce="0,0,0") {
    angularDampingOverride=0.999
    angularSpringConstant=8000
    bAngularSpring=true
    bOverrideAngularDamping=true
    bUseGravityOverride=true
    gravityScale=0.35
    linearDampingOverride=0.999
    linearSpringConstant=100
    simulationSpace=World
    alphaBoolBlend {blendOption=Linear}
    boundBone {boneName=dyn_fr_bagginess_c}
    componentAppliedLinearAccClamp=<3000 3000 3000>
    componentLinearAccScale=<1 1 1>
    componentLinearVelScale=<1 1 1>
    gravityOverride {z=-980}
    physicsBodyDefinitions [{
      boundBone {boneName=dyn_fr_bagginess_c}
      boxExtents=<30 30 30>
      constraintSetup {angularLimitsMax=<0 10 5>, angularLimitsMin=<0 -10 -5>, angularTarget=<1 0 0>}
      localJointOffset=<1 0 0>
     }]
    relativeSpaceBone {boneName=pelvis}
   }
   n14 = Anim Dynamics - Bone: dyn_sd_bagginess_l(ComponentPose=@n13, Alpha=1, GravityOverride="0,0,-980", ExternalForce="0,0,0") {
    angularDampingOverride=1
    angularSpringConstant=7000
    bAngularSpring=true
    bOverrideAngularDamping=true
    bUseGravityOverride=true
    gravityScale=0.35
    linearDampingOverride=0.999
    linearSpringConstant=200
    simulationSpace=World
    alphaBoolBlend {blendOption=Linear}
    boundBone {boneName=dyn_sd_bagginess_l}
    componentAppliedLinearAccClamp=<3000 3000 3000>
    componentLinearAccScale=<0.45 0.45 0.45>
    componentLinearVelScale=<0.45 0.45 0.45>
    gravityOverride {z=-980}
    physicsBodyDefinitions [{
      boundBone {boneName=dyn_sd_bagginess_l}
      boxExtents=<30 30 30>
      constraintSetup {angularLimitsMax=<10 30 10>, angularLimitsMin=<-10 -30 -10>, angularTarget=<1 0 0>}
      localJointOffset=<1.5 0 0>
     }]
    relativeSpaceBone {boneName=root}
   }
   n15 = Anim Dynamics - Bone: dyn_sd_bagginess_r(ComponentPose=@n14, Alpha=1, GravityOverride="0,0,-980", ExternalForce="0,0,0") {
    angularDampingOverride=1
    angularSpringConstant=7000
    bAngularSpring=true
    bOverrideAngularDamping=true
    bUseGravityOverride=true
    gravityScale=0.35
    linearDampingOverride=0.999
    linearSpringConstant=200
    simulationSpace=World
    alphaBoolBlend {blendOption=Linear}
    boundBone {boneName=dyn_sd_bagginess_r}
    componentAppliedLinearAccClamp=<3000 3000 3000>
    componentLinearAccScale=<0.45 0.45 0.45>
    componentLinearVelScale=<0.45 0.45 0.45>
    gravityOverride {z=-980}
    physicsBodyDefinitions [{
      boundBone {boneName=dyn_sd_bagginess_r}
      boxExtents=<30 30 30>
      constraintSetup {angularLimitsMax=<10 30 10>, angularLimitsMin=<-10 -30 -10>, angularTarget=<1 0 0>}
      localJointOffset=<1.5 0 0>
     }]
    relativeSpaceBone {boneName=root}
   }
   n16 = Anim Dynamics - Bone: dyn_bk_bagginess_c(ComponentPose=@n15, Alpha=1, GravityOverride="0,0,-980", ExternalForce="0,0,0") {
    angularDampingOverride=1
    angularSpringConstant=7000
    bAngularSpring=true
    bLinearSpring=true
    bOverrideAngularDamping=true
    bOverrideLinearDamping=true
    bUseGravityOverride=true
    gravityScale=0.35
    linearDampingOverride=0.999
    linearSpringConstant=200
    simulationSpace=World
    alphaBoolBlend {blendOption=Linear}
    boundBone {boneName=dyn_bk_bagginess_c}
    componentAppliedLinearAccClamp=<3000 3000 3000>
    componentLinearAccScale=<0.45 0.45 0.45>
    componentLinearVelScale=<0.45 0.45 0.45>
    gravityOverride {z=-980}
    physicsBodyDefinitions [{
      boundBone {boneName=dyn_bk_bagginess_c}
      boxExtents=<30 30 30>
      constraintSetup {angularLimitsMax=<10 10 10>, angularLimitsMin=<-10 -10 -10>, angularTarget=<1 0 0>, linearAxesMax=<0 0 3>}
      localJointOffset=<1.5 0 0>
     }]
    relativeSpaceBone {boneName=root}
   }
   n17 = Anim Dynamics - Bone: dyn_pin_a_01_l!(SimulationSpace:AnimPhysSimSpaceType Enum=Component, ComponentPose=@n16, Alpha=0.5, GravityOverride:Vector="0,0,-980", ExternalForce="0,0,0") -> (Pose) {
    angularDampingOverride=1
    angularSpringConstant=20000
    bAngularSpring=true
    bOverrideAngularDamping=true
    bUseGravityOverride=true
    gravityScale=0.35
    simulationSpace=RootRelative
    alphaBoolBlend {blendOption=Linear}
    boundBone {boneName=dyn_pin_a_01_l}
    componentAppliedLinearAccClamp=<3000 3000 3000>
    componentLinearAccScale=<1 1 1>
    componentLinearVelScale=<1 1 1>
    gravityOverride {z=-980}
    physicsBodyDefinitions [{
      boundBone {boneName=dyn_pin_a_01_l}
      boxExtents=<20 30 30>
      constraintSetup {angularLimitsMax=<5 15 15>, angularLimitsMin=<-5 -15 -15>, angularTarget=<1 0 0>}
      localJointOffset=<7 0 0>
     }]
   }
   n18 = Anim Dynamics - Bone: dyn_pin_a_02_l!(SimulationSpace:AnimPhysSimSpaceType Enum, ComponentPose=@n17, Alpha=0.5, GravityOverride:Vector="0,0,-980", ExternalForce="0,0,0") -> (Pose) {
    angularDampingOverride=1
    angularSpringConstant=20000
    bAngularSpring=true
    bOverrideAngularDamping=true
    bUseGravityOverride=true
    gravityScale=0.35
    simulationSpace=RootRelative
    alphaBoolBlend {blendOption=Linear}
    boundBone {boneName=dyn_pin_a_02_l}
    componentAppliedLinearAccClamp=<3000 3000 3000>
    componentLinearAccScale=<1 1 1>
    componentLinearVelScale=<1 1 1>
    gravityOverride {z=-980}
    physicsBodyDefinitions [{
      boundBone {boneName=dyn_pin_a_02_l}
      boxExtents=<20 30 30>
      constraintSetup {angularLimitsMax=<5 15 15>, angularLimitsMin=<-5 -15 -15>, angularTarget=<1 0 0>}
      localJointOffset=<7 0 0>
     }]
   }
   n19 = Anim Dynamics - Bone: dyn_pin_b_01_l!(SimulationSpace:AnimPhysSimSpaceType Enum, ComponentPose=@n18, Alpha=0.5, GravityOverride:Vector="0,0,-980", ExternalForce="0,0,0") -> (Pose) {
    angularDampingOverride=1
    angularSpringConstant=20000
    bAngularSpring=true
    bOverrideAngularDamping=true
    bUseGravityOverride=true
    gravityScale=0.35
    simulationSpace=RootRelative
    alphaBoolBlend {blendOption=Linear}
    boundBone {boneName=dyn_pin_b_01_l}
    componentAppliedLinearAccClamp=<3000 3000 3000>
    componentLinearAccScale=<1 1 1>
    componentLinearVelScale=<1 1 1>
    gravityOverride {z=-980}
    physicsBodyDefinitions [{
      boundBone {boneName=dyn_pin_b_01_l}
      boxExtents=<20 30 30>
      constraintSetup {angularLimitsMax=<5 15 15>, angularLimitsMin=<-5 -15 -15>, angularTarget=<1 0 0>}
      localJointOffset=<7 0 0>
     }]
   }
   n20 = Anim Dynamics - Bone: dyn_pin_b_02_l!(SimulationSpace:AnimPhysSimSpaceType Enum, ComponentPose=@n19, Alpha=0.5, GravityOverride:Vector="0,0,-980", ExternalForce="0,0,0") -> (Pose) {
    angularDampingOverride=1
    angularSpringConstant=20000
    bAngularSpring=true
    bOverrideAngularDamping=true
    bUseGravityOverride=true
    gravityScale=0.35
    simulationSpace=RootRelative
    alphaBoolBlend {blendOption=Linear}
    boundBone {boneName=dyn_pin_b_02_l}
    componentAppliedLinearAccClamp=<3000 3000 3000>
    componentLinearAccScale=<1 1 1>
    componentLinearVelScale=<1 1 1>
    gravityOverride {z=-980}
    physicsBodyDefinitions [{
      boundBone {boneName=dyn_pin_b_02_l}
      boxExtents=<20 30 30>
      constraintSetup {angularLimitsMax=<5 15 15>, angularLimitsMin=<-5 -15 -15>, angularTarget=<1 0 0>}
      localJointOffset=<7 0 0>
     }]
   }
   n21 = Anim Dynamics - Bone: dyn_pin_c_01_l!(SimulationSpace:AnimPhysSimSpaceType Enum, ComponentPose=@n20, Alpha=0.5, GravityOverride:Vector="0,0,-980", ExternalForce="0,0,0") -> (Pose) {
    angularDampingOverride=1
    angularSpringConstant=25000
    bAngularSpring=true
    bOverrideAngularDamping=true
    bUseGravityOverride=true
    gravityScale=0.35
    simulationSpace=RootRelative
    alphaBoolBlend {blendOption=Linear}
    boundBone {boneName=dyn_pin_c_01_l}
    componentAppliedLinearAccClamp=<3000 3000 3000>
    componentLinearAccScale=<1 1 1>
    componentLinearVelScale=<1 1 1>
    gravityOverride {z=-980}
    physicsBodyDefinitions [{
      boundBone {boneName=dyn_pin_c_01_l}
      boxExtents=<20 30 30>
      constraintSetup {angularLimitsMax=<5 5 20>, angularLimitsMin=<-5 -15 -10>, angularTarget=<1 0 0>}
      localJointOffset=<6 0 0>
     }]
   }
   n22 = Anim Dynamics - Bone: dyn_pin_c_02_l!(SimulationSpace:AnimPhysSimSpaceType Enum, ComponentPose=@n21, Alpha=0.5, GravityOverride:Vector="0,0,-980", ExternalForce="0,0,0") -> (Pose) {
    angularDampingOverride=1
    angularSpringConstant=25000
    bAngularSpring=true
    bOverrideAngularDamping=true
    bUseGravityOverride=true
    gravityScale=0.35
    simulationSpace=RootRelative
    alphaBoolBlend {blendOption=Linear}
    boundBone {boneName=dyn_pin_c_02_l}
    componentAppliedLinearAccClamp=<3000 3000 3000>
    componentLinearAccScale=<1 1 1>
    componentLinearVelScale=<1 1 1>
    gravityOverride {z=-980}
    physicsBodyDefinitions [{
      boundBone {boneName=dyn_pin_c_02_l}
      boxExtents=<20 30 30>
      constraintSetup {angularLimitsMax=<5 5 20>, angularLimitsMin=<-5 -15 -10>, angularTarget=<1 0 0>}
      localJointOffset=<6 0 0>
     }]
   }
   n23 = Anim Dynamics - Bone: dyn_pin_d_01_l!(SimulationSpace:AnimPhysSimSpaceType Enum, ComponentPose=@n22, Alpha=0.5, GravityOverride:Vector="0,0,-980", ExternalForce="0,0,0") -> (Pose) {
    angularDampingOverride=1
    angularSpringConstant=25000
    bAngularSpring=true
    bOverrideAngularDamping=true
    bUseGravityOverride=true
    gravityScale=0.35
    simulationSpace=RootRelative
    alphaBoolBlend {blendOption=Linear}
    boundBone {boneName=dyn_pin_d_01_l}
    componentAppliedLinearAccClamp=<3000 3000 3000>
    componentLinearAccScale=<1 1 1>
    componentLinearVelScale=<1 1 1>
    gravityOverride {z=-980}
    physicsBodyDefinitions [{
      boundBone {boneName=dyn_pin_d_01_l}
      boxExtents=<20 30 30>
      constraintSetup {angularLimitsMax=<5 0 10>, angularLimitsMin=<-5 -15 -20>, angularTarget=<1 0 0>}
      localJointOffset=<6 0 0>
     }]
   }
   n24 = Anim Dynamics - Bone: dyn_pin_d_02_l!(SimulationSpace:AnimPhysSimSpaceType Enum, ComponentPose=@n23, Alpha=0.5, GravityOverride:Vector="0,0,-980", ExternalForce="0,0,0") -> (Pose) {
    angularDampingOverride=1
    angularSpringConstant=25000
    bAngularSpring=true
    bOverrideAngularDamping=true
    bUseGravityOverride=true
    gravityScale=0.35
    simulationSpace=RootRelative
    alphaBoolBlend {blendOption=Linear}
    boundBone {boneName=dyn_pin_d_02_l}
    componentAppliedLinearAccClamp=<3000 3000 3000>
    componentLinearAccScale=<1 1 1>
    componentLinearVelScale=<1 1 1>
    gravityOverride {z=-980}
    physicsBodyDefinitions [{
      boundBone {boneName=dyn_pin_d_02_l}
      boxExtents=<20 30 30>
      constraintSetup {angularLimitsMax=<5 0 10>, angularLimitsMin=<-5 -15 -20>, angularTarget=<1 0 0>}
      localJointOffset=<6 0 0>
     }]
   }
   n25 = Anim Dynamics - Bone: dyn_pin_a_01_r!(SimulationSpace:AnimPhysSimSpaceType Enum, ComponentPose=@n24, Alpha=0.5, GravityOverride:Vector="0,0,-980", ExternalForce="0,0,0") -> (Pose) {
    angularDampingOverride=1
    angularSpringConstant=20000
    bAngularSpring=true
    bOverrideAngularDamping=true
    bUseGravityOverride=true
    gravityScale=0.35
    simulationSpace=RootRelative
    alphaBoolBlend {blendOption=Linear}
    boundBone {boneName=dyn_pin_a_01_r}
    componentAppliedLinearAccClamp=<3000 3000 3000>
    componentLinearAccScale=<1 1 1>
    componentLinearVelScale=<1 1 1>
    gravityOverride {z=-980}
    physicsBodyDefinitions [{
      boundBone {boneName=dyn_pin_a_01_r}
      boxExtents=<20 30 30>
      constraintSetup {angularLimitsMax=<5 5 15>, angularLimitsMin=<-5 -15 -15>, angularTarget=<1 0 0>}
      localJointOffset=<7 0 0>
     }]
   }
   n26 = Anim Dynamics - Bone: dyn_pin_a_02_r!(SimulationSpace:AnimPhysSimSpaceType Enum, ComponentPose=@n25, Alpha=0.5, GravityOverride:Vector="0,0,-980", ExternalForce="0,0,0") -> (Pose) {
    angularDampingOverride=1
    angularSpringConstant=20000
    bAngularSpring=true
    bOverrideAngularDamping=true
    bUseGravityOverride=true
    gravityScale=0.35
    simulationSpace=RootRelative
    alphaBoolBlend {blendOption=Linear}
    boundBone {boneName=dyn_pin_a_02_r}
    componentAppliedLinearAccClamp=<3000 3000 3000>
    componentLinearAccScale=<1 1 1>
    componentLinearVelScale=<1 1 1>
    gravityOverride {z=-980}
    physicsBodyDefinitions [{
      boundBone {boneName=dyn_pin_a_02_r}
      boxExtents=<20 30 30>
      constraintSetup {angularLimitsMax=<5 5 15>, angularLimitsMin=<-5 -15 -15>, angularTarget=<1 0 0>}
      localJointOffset=<7 0 0>
     }]
   }
   n27 = Anim Dynamics - Bone: dyn_pin_b_01_r!(SimulationSpace:AnimPhysSimSpaceType Enum, ComponentPose=@n26, Alpha=0.5, GravityOverride:Vector="0,0,-980", ExternalForce="0,0,0") -> (Pose) {
    angularDampingOverride=1
    angularSpringConstant=20000
    bAngularSpring=true
    bOverrideAngularDamping=true
    bUseGravityOverride=true
    gravityScale=0.35
    simulationSpace=RootRelative
    alphaBoolBlend {blendOption=Linear}
    boundBone {boneName=dyn_pin_b_01_r}
    componentAppliedLinearAccClamp=<3000 3000 3000>
    componentLinearAccScale=<1 1 1>
    componentLinearVelScale=<1 1 1>
    gravityOverride {z=-980}
    physicsBodyDefinitions [{
      boundBone {boneName=dyn_pin_b_01_r}
      boxExtents=<20 30 30>
      constraintSetup {angularLimitsMax=<5 5 15>, angularLimitsMin=<-5 -15 -15>, angularTarget=<1 0 0>}
      localJointOffset=<7 0 0>
     }]
   }
   n28 = Anim Dynamics - Bone: dyn_pin_b_02_r!(SimulationSpace:AnimPhysSimSpaceType Enum, ComponentPose=@n27, Alpha=0.5, GravityOverride:Vector="0,0,-980", ExternalForce="0,0,0") -> (Pose) {
    angularDampingOverride=1
    angularSpringConstant=20000
    bAngularSpring=true
    bOverrideAngularDamping=true
    bUseGravityOverride=true
    gravityScale=0.35
    simulationSpace=RootRelative
    alphaBoolBlend {blendOption=Linear}
    boundBone {boneName=dyn_pin_b_02_r}
    componentAppliedLinearAccClamp=<3000 3000 3000>
    componentLinearAccScale=<1 1 1>
    componentLinearVelScale=<1 1 1>
    gravityOverride {z=-980}
    physicsBodyDefinitions [{
      boundBone {boneName=dyn_pin_b_02_r}
      boxExtents=<20 30 30>
      constraintSetup {angularLimitsMax=<5 5 15>, angularLimitsMin=<-5 -15 -15>, angularTarget=<1 0 0>}
      localJointOffset=<7 0 0>
     }]
   }
   n29 = Anim Dynamics - Bone: dyn_pin_c_01_r!(SimulationSpace:AnimPhysSimSpaceType Enum, ComponentPose=@n28, Alpha=0.5, GravityOverride:Vector="0,0,-980", ExternalForce="0,0,0") -> (Pose) {
    angularDampingOverride=1
    angularSpringConstant=20000
    bAngularSpring=true
    bOverrideAngularDamping=true
    bUseGravityOverride=true
    gravityScale=0.35
    simulationSpace=RootRelative
    alphaBoolBlend {blendOption=Linear}
    boundBone {boneName=dyn_pin_c_01_r}
    componentAppliedLinearAccClamp=<3000 3000 3000>
    componentLinearAccScale=<1 1 1>
    componentLinearVelScale=<1 1 1>
    gravityOverride {z=-980}
    physicsBodyDefinitions [{
      boundBone {boneName=dyn_pin_c_01_r}
      boxExtents=<20 30 30>
      constraintSetup {angularLimitsMax=<5 5 20>, angularLimitsMin=<-5 -15 -10>, angularTarget=<1 0 0>}
      localJointOffset=<6 0 0>
     }]
   }
   n30 = Anim Dynamics - Bone: dyn_pin_c_02_r!(SimulationSpace:AnimPhysSimSpaceType Enum, ComponentPose=@n29, Alpha=0.5, GravityOverride:Vector="0,0,-980", ExternalForce="0,0,0") -> (Pose) {
    angularDampingOverride=1
    angularSpringConstant=20000
    bAngularSpring=true
    bOverrideAngularDamping=true
    bUseGravityOverride=true
    gravityScale=0.35
    simulationSpace=RootRelative
    alphaBoolBlend {blendOption=Linear}
    boundBone {boneName=dyn_pin_c_02_r}
    componentAppliedLinearAccClamp=<3000 3000 3000>
    componentLinearAccScale=<1 1 1>
    componentLinearVelScale=<1 1 1>
    gravityOverride {z=-980}
    physicsBodyDefinitions [{
      boundBone {boneName=dyn_pin_c_02_r}
      boxExtents=<20 30 30>
      constraintSetup {angularLimitsMax=<5 5 20>, angularLimitsMin=<-5 -15 -10>, angularTarget=<1 0 0>}
      localJointOffset=<6 0 0>
     }]
   }
   n31 = Anim Dynamics - Bone: dyn_pin_d_01_r!(SimulationSpace:AnimPhysSimSpaceType Enum, ComponentPose=@n30, Alpha=0.5, GravityOverride:Vector="0,0,-980", ExternalForce="0,0,0") -> (Pose) {
    angularDampingOverride=1
    angularSpringConstant=20000
    bAngularSpring=true
    bOverrideAngularDamping=true
    bUseGravityOverride=true
    gravityScale=0.35
    simulationSpace=RootRelative
    alphaBoolBlend {blendOption=Linear}
    boundBone {boneName=dyn_pin_d_01_r}
    componentAppliedLinearAccClamp=<3000 3000 3000>
    componentLinearAccScale=<1 1 1>
    componentLinearVelScale=<1 1 1>
    gravityOverride {z=-980}
    physicsBodyDefinitions [{
      boundBone {boneName=dyn_pin_d_01_r}
      boxExtents=<20 30 30>
      constraintSetup {angularLimitsMax=<5 0 20>, angularLimitsMin=<-5 -15 -10>, angularTarget=<1 0 0>}
      localJointOffset=<6 0 0>
     }]
   }
   n32 = Anim Dynamics - Bone: dyn_pin_d_02_r!(SimulationSpace:AnimPhysSimSpaceType Enum, ComponentPose=@n31, Alpha=0.5, GravityOverride:Vector="0,0,-980", ExternalForce="0,0,0") -> (Pose) {
    angularDampingOverride=1
    angularSpringConstant=20000
    bAngularSpring=true
    bOverrideAngularDamping=true
    bUseGravityOverride=true
    gravityScale=0.35
    simulationSpace=RootRelative
    alphaBoolBlend {blendOption=Linear}
    boundBone {boneName=dyn_pin_d_02_r}
    componentAppliedLinearAccClamp=<3000 3000 3000>
    componentLinearAccScale=<1 1 1>
    componentLinearVelScale=<1 1 1>
    gravityOverride {z=-980}
    physicsBodyDefinitions [{
      boundBone {boneName=dyn_pin_d_02_r}
      boxExtents=<20 30 30>
      constraintSetup {angularLimitsMax=<5 0 20>, angularLimitsMin=<-5 -15 -10>, angularTarget=<1 0 0>}
      localJointOffset=<6 0 0>
     }]
   }
   n33 = "Transform (Modify) Bone - Bone: dyn_necklees_02"(ComponentPose=@n32, Alpha=1, Translation="0,0,0", Rotation="28.705,0.000,-0.000", Scale="1,1,1") {
    rotationMode=BMM_Additive
    rotationSpace=BCS_BoneSpace
    translationMode=BMM_Additive
    translationSpace=BCS_BoneSpace
    alphaBoolBlend {blendInTime=0.2, blendOption=Linear, blendOutTime=0.2}
    alphaScaleBiasClamp {bInterpResult=true}
    boneToModify {boneName=dyn_necklees_02}
    rotation=<rot 28.705 0 0>
   }
   Output Pose(Result=Component To Local(ComponentPose=@n33)) {layerGroup=None}
   n34 = Control Rig!(
    Source=Copy Pose From Mesh() {bUseAttachedParent=true}
    Alpha
    Detected_Pistol:double=Get DetectedPistol() -> (DetectedPistol:double)
    No Skydiving:double=Get No Skydiving() -> (No Skydiving:double)
    HangGliding:bool=Get HangGlidingAlpha() -> (HangGlidingAlpha:bool)
   ) -> (Pose) {
    alphaBoolBlend {blendOption=Linear}
    controlRigAssetReference {blueprintRigClass=&/M_MED_PineTrimPack_Sift_CtrlRig.M_MED_PineTrimPack_Sift_CtrlRig_C}
   }
   Save cached pose 'CtrlRigCache'(Pose=@n34) {cacheName=CtrlRigCache}
   n35 = "Transform (Modify) Bone - Bone: base_necklees"(ComponentPose=Local To Component(LocalPose=Use cached pose 'CtrlRigCache'()), Alpha=1, Translation="0,0,0", Rotation="20,0,0", Scale="1,1,1") {
    rotationMode=BMM_Additive
    rotationSpace=BCS_BoneSpace
    alphaBoolBlend {blendOption=Linear}
    alphaScaleBiasClamp {bInterpResult=true, interpSpeedIncreasing=2.5}
    boneToModify {boneName=base_necklees}
    rotation {pitch=20}
   }
   n36 = "Transform (Modify) Bone - Bone: dyn_necklees_02"(ComponentPose=@n35, Alpha=1, Translation="0,0,0", Rotation="30,0,0", Scale="1,1,1") {
    rotationMode=BMM_Additive
    rotationSpace=BCS_BoneSpace
    alphaBoolBlend {blendOption=Linear}
    alphaScaleBiasClamp {bInterpResult=true, interpSpeedIncreasing=2.5}
    boneToModify {boneName=dyn_necklees_02}
    rotation {pitch=30}
   }
   n37 = "Transform (Modify) Bone - Bone: base_hood_02_dn"(ComponentPose=@n36, Alpha=1, Translation="0,0,0", Rotation="30,0,0", Scale="1,1,1") {
    rotationMode=BMM_Additive
    rotationSpace=BCS_BoneSpace
    alphaBoolBlend {blendOption=Linear}
    alphaScaleBiasClamp {bInterpResult=true, interpSpeedIncreasing=2.5}
    boneToModify {boneName=base_hood_02_dn}
    rotation {pitch=30}
   }
   n38 = "Transform (Modify) Bone - Bone: dyn_necklees_01"!(ComponentPose=@n37, bAlphaBoolEnabled:bool=True, Translation="0,0,0", Rotation="1.5,0,0", Scale="1,1,1") -> (Pose) {
    alphaInputType=Bool
    rotationMode=BMM_Additive
    rotationSpace=BCS_BoneSpace
    alphaBoolBlend {blendInTime=0.1, blendOption=Linear, blendOutTime=0.1}
    alphaScaleBiasClamp {bInterpResult=true, interpSpeedIncreasing=2.5}
    boneToModify {boneName=dyn_necklees_01}
    rotation {pitch=1.5}
   }
   n39 = "Transform (Modify) Bone - Bone: attach_cape"!(ComponentPose=@n38, bAlphaBoolEnabled:bool=True, Translation="0,-1.2,0", Rotation="0,0,0", Scale="1,1,1") -> (Pose) {
    alphaInputType=Bool
    rotationMode=BMM_Additive
    rotationSpace=BCS_BoneSpace
    translationMode=BMM_Additive
    translationSpace=BCS_BoneSpace
    alphaBoolBlend {blendInTime=0.1, blendOption=Linear, blendOutTime=0.1}
    alphaScaleBiasClamp {bInterpResult=true, interpSpeedIncreasing=2.5}
    boneToModify {boneName=attach_cape}
    translation {y=-1.2}
   }
   n40 = "Transform (Modify) Bone - Bone: base_hood_dn"(ComponentPose=@n39, Alpha=1, Translation="2.961,-0.000,3.950", Rotation="-12.336,-0.000,-0.000", Scale="1,1,1") {
    rotationMode=BMM_Additive
    rotationSpace=BCS_BoneSpace
    translationMode=BMM_Additive
    translationSpace=BCS_BoneSpace
    alphaBoolBlend {blendOption=Linear}
    alphaScaleBiasClamp {bInterpResult=true}
    boneToModify {boneName=base_hood_dn}
    rotation=<rot -12.336 0 0>
    translation {x=2.961, z=3.95}
   }
   n41 = "Transform (Modify) Bone - Bone: dyn_zip_02"(ComponentPose=@n40, Alpha=1, Translation="0,0,0", Rotation="77.636,-0.002,0.000", Scale="1,1,1") {
    rotationMode=BMM_Additive
    rotationSpace=BCS_BoneSpace
    translationMode=BMM_Additive
    translationSpace=BCS_BoneSpace
    alphaBoolBlend {blendInTime=0.2, blendOption=Linear, blendOutTime=0.2}
    alphaScaleBiasClamp {bInterpResult=true}
    boneToModify {boneName=dyn_zip_02}
    rotation=<rot 77.636 -0.002 0>
   }
   n42 = "Transform (Modify) Bone - Bone: base_necklees"(ComponentPose=@n41, Alpha=1, Translation="0,0,0", Rotation="21.815,0.001,0.000", Scale="1,1,1") {
    rotationMode=BMM_Additive
    rotationSpace=BCS_BoneSpace
    translationMode=BMM_Additive
    translationSpace=BCS_BoneSpace
    alphaBoolBlend {blendInTime=0.2, blendOption=Linear, blendOutTime=0.2}
    alphaScaleBiasClamp {bInterpResult=true}
    boneToModify {boneName=base_necklees}
    rotation=<rot 21.815 0.001 0>
   }
   n43 = "Transform (Modify) Bone - Bone: attach_cape"!(ComponentPose=@n42, Alpha:float=1, Translation="0,1,-2", Rotation="0,0,0", Scale="1,1,1") -> (Pose) {
    rotationMode=BMM_Additive
    rotationSpace=BCS_BoneSpace
    translationMode=BMM_Additive
    translationSpace=BCS_BoneSpace
    alphaBoolBlend {blendInTime=0.1, blendOption=Linear, blendOutTime=0.1}
    alphaScaleBiasClamp {bInterpResult=true, interpSpeedIncreasing=2.5}
    boneToModify {boneName=attach_cape}
    translation {y=1, z=-2}
   }
   n44 = "Transform (Modify) Bone - Bone: attach_backpack"(ComponentPose=@n43, Alpha=1, Translation="0,0.8,0", Rotation="0,0,0", Scale="1,1,1") {
    rotationMode=BMM_Additive
    rotationSpace=BCS_BoneSpace
    translationMode=BMM_Additive
    translationSpace=BCS_BoneSpace
    alphaBoolBlend {blendInTime=0.1, blendOption=Linear, blendOutTime=0.1}
    alphaScaleBiasClamp {bInterpResult=true, interpSpeedIncreasing=2.5}
    boneToModify {boneName=attach_backpack}
    translation {y=0.8}
   }
   Save cached pose 'AdjustCache'(Pose=Component To Local(ComponentPose=@n44)) {cacheName=AdjustCache}
  ]}
 EventGraph {nodes [
   Event BlueprintInitializeAnimation()
   > Set SkelMesh(SkelMesh=GetOwnerSkeletalMeshComponent())
   n1 = TryGetPawnOwner()
   > Is Valid(InputObject=@n1)
   > n2 = Cast To FortPlayerPawnAthena(Object=@n1)
   > n3 = Set PlayerPawn(PlayerPawn:Fort Player Pawn Athena Object Reference=@n2) -> (Output_Get:Fort Player Pawn Athena Object Reference)
   > Set Body_Mesh(Body_Mesh=GetSkeletalMeshForPartType(self=@n3))
   n4 = Event BlueprintUpdateAnimation()
   > Set DeltaTimeX(DeltaTimeX:double=@n4.DeltaTimeX) -> (Output_Get:double)
   > n5 = Sequence!(execute) -> (then_0, then_1, then_2:Exec, then_3:Exec, then_4:Exec, then_5:Exec, then_6:Exec, then_7:Exec)
   > Sequence!(execute) -> (then_0, then_1)
   > Is Valid(InputObject=Get PlayerPawn() -> (PlayerPawn:Fort Player Pawn Athena Object Reference))
   > n6 = Branch(Condition=Get bIsInFrontEnd())
   Set WPO Globals Function(
    execute=@n6.else
    Material=Get Mat Instance Dynamic()
    Overlay Material=Get Overlay Mat Instance Dynamic()
    Skeletal Mesh=Get SkelMesh()
    "Local Noise Transform  Location"=Get Local Noise Transform Location() -> (Local Noise Transform Location:Vector)
    Local Noise Offset=Get Local Noise Offset()
   )
   n7 = Get DeltaTimeX() -> (DeltaTimeX:double)
   > Set WPO Regional Function!(
    execute
    Material:Material Instance Dynamic Object Reference=Get Mat Instance Dynamic()
    Overlay Material:Material Instance Dynamic Object Reference=Get Overlay Mat Instance Dynamic()
    Skeletal Mesh:Skeletal Mesh Component Object Reference=Get SkelMesh()
    Velocity Transform:Name=foot_r
    Velocity Interp Speed:double=Get Velocity Interp Speed()
    Collision Transform:Name=calf_r
    Collision Transform Axis:EAxis Enum
    Invert Collision Transform Axis:bool=false
    Delta Time X:double=@n7
    Previous Velocity Transform Location:"Vector (by ref)"=Get R1 Velocity Transform Location()
    Previous Velocity Transform Velocity:"Vector (by ref)"=Get R1 Velocity Transform Velocity()
    Region:int=1
   ) -> (then)
   > Set WPO Regional Function(
    Material=Get Mat Instance Dynamic()
    Overlay Material=Get Overlay Mat Instance Dynamic()
    Skeletal Mesh=Get SkelMesh()
    Velocity Transform=lowerarm_r
    Velocity Interp Speed=Get Velocity Interp Speed()
    Collision Transform=upperarm_r
    Invert Collision Transform Axis=true
    Delta Time X=@n7
    Previous Velocity Transform Location=Get R2 Velocity Transform Location()
    Previous Velocity Transform Velocity=Get R2 Velocity Transform Velocity()
    Region=2
   )
   > n8 = Branch(Condition=OR Boolean(A=Get bIsSkydiveFloating(), B=Get bIsParachuteOpened()))
   > Set WPO Pho Wind Intensity(WPO Pho Wind Intensity=0.4)
   > n9 = Set WPO Pho Turbulence(WPO Pho Turbulence:double=0.85) -> (Output_Get:double)
   SetScalarParameterValue(
    execute=[@n9.then, @n11.then]
    self=[Get Mat Instance Dynamic(), Get Overlay Mat Instance Dynamic()]
    ParameterName=ClothSim_WindStrength
    Value=Get WPO Pho Wind Intensity()
   )
   n10 = Get WPO Pho Turbulence() -> (WPO Pho Turbulence:double)
   > SetScalarParameterValue(self=Get Mat Instance Dynamic(), ParameterName=ClothSim_R1_TurbulenceStrength, Value=@n10)
   > SetScalarParameterValue(self=Get Mat Instance Dynamic(), ParameterName=ClothSim_R2_TurbulenceStrength, Value=@n10)
   Set WPO Pho Wind Intensity(execute=@n8.else, WPO Pho Wind Intensity=0)
   > n11 = Set WPO Pho Turbulence(WPO Pho Turbulence=0)
   n12 = "Gravity Override State Function (Parameter Asset) (Nativized)"(
    execute=@n5.then_1
    CharacterPartAnimInstance=Self-Reference()
    GameTimeInSeconds=GetGameTimeInSeconds()
    DebugJointName=spine_05
    Parameters=&/M_MED_PineTrimPack_Sift_Parameters
   )
   > Set Gravity Override(Gravity Override=@n12)
   > Set RBAN Settings Out(RBAN Settings Out=@n12.OutRigidBodyAnimNodeParameters)
   n13 = Get bIsDBNOCarrying() -> (bIsDBNOCarrying:bool)
   n14 = Get bIsDBNOCarried()
   n15 = Branch(execute=@n5.then_4, Condition=AND Boolean!(A=Get bIsDBNO(), B=NOT Boolean(A=@n14), C:bool=NOT Boolean(A=@n13)) -> (ReturnValue))
   > Set DBNOGround(DBNOGround:double=1) -> (Output_Get:double)
   > Set NoDBNO(NoDBNO:double=0) -> (Output_Get:double)
   Set DBNOGround(execute=@n15.else, DBNOGround=0)
   > Set NoDBNO(NoDBNO=1)
   n16 = Branch(execute=@n5.then_5, Condition=Get bIsInFrontEnd())
   > Set Simulation Space(Simulation Space:AnimPhysSimSpaceType Enum=World) -> (Output_Get:AnimPhysSimSpaceType Enum)
   > Set FrontEnd(FrontEnd:double=0.5) -> (Output_Get:double)
   Set Simulation Space(execute=@n16.else)
   > Set FrontEnd(FrontEnd=1)
   n17 = Cast To FortPlayerPawn(execute=@n5.then_7, Object=TryGetPawnOwner())
   n18 = MatchesWildcard(
    SourceString=GetDisplayName(Object=GetSkeletalMeshForPartType(self=@n17, PartType=Face))
    Wildcard=*M_MED_PineTrimPack_Sift_FaceAcc*
    SearchCase=CaseSensitive
   )
   n19 = Get bIsInFrontEnd()
   > Set Detected FA(Detected FA:double="To Float (Boolean)"(InBool=AND Boolean!(A=@n18, B=@n19) -> (ReturnValue))) -> (Output_Get:double)
   n20 = Get bIsSkyDiving()
   Set No Skydiving(execute=@n5.then_2, No Skydiving:double="To Float (Boolean)"(InBool=NOT Boolean(A=@n20))) -> (Output_Get:double)
   n21 = NOT Boolean(A=Get bIsSurfaceSwimming())
   n22 = NOT Boolean(A=Get bIsSkyDiving())
   n23 = NOT Boolean(A=Get bIsPlayingEmote())
   n24 = Get bIsFalling()
   > Set FallingOnGround(FallingOnGround:double="To Float (Boolean)"(InBool=AND Boolean!(A=@n24, B=@n22, C:bool=@n23, D:bool=@n21) -> (ReturnValue))) -> (Output_Get:double)
   n25 = Get bIsSkyDiving()
   n26 = Get bIsPlayingEmote()
   n27 = Get bIsMoving2D()
   > n28 = Branch(Condition=AND Boolean(A=Get bIsInFrontEnd(), B=NOT Boolean(A=@n27), C=NOT Boolean(A=@n25), D=NOT Boolean(A=@n26)))
   > Set FE_Idle(FE_Idle:double=1) -> (Output_Get:double)
   > Set FE_NotFEIdle(FE_NotFEIdle:double=0) -> (Output_Get:double)
   Set FE_Idle(execute=@n28.else, FE_Idle=0)
   > Set FE_NotFEIdle(FE_NotFEIdle=1)
   n29 = Get bIsMoving2D()
   n30 = Get bIsSkyDiving()
   n31 = Get bIsPlayingEmote()
   n32 = Branch(execute=@n5.then_3, Condition=AND Boolean(A=Get bIsInFrontEnd(), B=NOT Boolean(A=@n29), C=NOT Boolean(A=@n30), D=NOT Boolean(A=@n31)))
   > Set FE_Idle_Necklees(FE_Idle_Necklees:double=0.3) -> (Output_Get:double)
   > Set FE_Not_Idle_Necklees(FE_Not_Idle_Necklees:double=0) -> (Output_Get:double)
   Set FE_Idle_Necklees(execute=@n32.else, FE_Idle_Necklees=0)
   > Set FE_Not_Idle_Necklees(FE_Not_Idle_Necklees=0.5)
   n33 = GetOwnerSkeletalMeshComponent()
   n34 = BreakRotator(InRot=BreakTransform(InTransform=GetDeltaTransformFromRefPose(self=@n33, BoneName=neck_02, BaseName=spine_05)).Rotation)
   n35 = Branch(execute=@n5.then_6, Condition=AND Boolean!(A="float < float"(A=@n34.Yaw, B=-30), B=Get bIsCrouching()) -> (ReturnValue))
   > Set Alpha_HeadRotateUpExtreme(Alpha_HeadRotateUpExtreme:double=1) -> (Output_Get:double)
   Set Alpha_HeadRotateUpExtreme(execute=@n35.else, Alpha_HeadRotateUpExtreme=0)
   Event OnParachuteOpened()
   > Switch on EGliderType(Selection=Get PlayerGliderType() -> (PlayerGliderType=HangGlider))
   > Set HangGlidingAlpha(HangGlidingAlpha:bool=1) -> (Output_Get:bool)
   Event OnParachuteClosed()
   > Set HangGlidingAlpha(HangGlidingAlpha=false)
   Event OnCharacterCustomizationLoaded()
   > n36 = Set SkelMesh(SkelMesh=GetPartSkeletalMeshComponent())
   > n37 = Get Dynamic Material Instance(SkeletalMeshComponent=@n36)
   > Set Mat Instance Dynamic(Mat Instance Dynamic=@n37)
   > n38 = GetOrCreateDynamicOverlayMaterial(MeshComponent=@n36, MaterialName=None)
   > Set Overlay Mat Instance Dynamic(Overlay Mat Instance Dynamic=@n38)
   Event OnWeaponChanged()
   > n39 = Cast To FortPlayerAnimInstance(Object=GetAnimInstance(self=GetOwnerSkeletalMeshComponent()))
   > Set DetectedPistol(DetectedPistol:double=0) -> (Output_Get:double)
   > n40 = Switch on EFortWeaponCoreAnimation!(execute, Selection:EFortWeaponCoreAnimation Enum=Get WeaponCoreAnim(self=@n39)) -> (
    Melee:Exec
    Pistol:Exec
    Shotgun:Exec
    PaperBlueprint:Exec
    Rifle:Exec
    MeleeOneHand:Exec
    MachinePistol:Exec
    RocketLauncher:Exec
    GrenadeLauncher:Exec
    GoingCommando:Exec
    AssaultRifle:Exec
    TacticalShotgun:Exec
    SniperRifle:Exec
    TrapPlacement:Exec
    ShoulderLauncher:Exec
    AbilityDecoTool:Exec
    Crossbow:Exec
    C4:Exec
    RemoteControl:Exec
    DualWield:Exec
    AR_BullPup:Exec
    AR_ForwardGrip:Exec
    MedPackPaddles:Exec
    SMG_P90:Exec
    AR_DrumGun:Exec
    Consumable_Small:Exec
    Consumable_Large:Exec
    Balloon:Exec
    MountedTurret:Exec
    CreativeTool:Exec
    ExplosiveBow:Exec
    AshtonIndigo:Exec
    AshtonChicago:Exec
    MeleeDualWield:Exec
    Unarmed:Exec
   )
   Set DetectedPistol(execute=@n40.Pistol, DetectedPistol=1)
  ]}
]
variables [
 Alpha_HeadRotateUpExtreme {type=number}
 Body_Mesh {type=SkeletalMeshComponent}
 DBNOGround {type=number}
 DeltaTimeX {type=number}
 Detected FA {type=number}
 DetectedPistol {type=number}
 FallingOnGround {type=number}
 FE_Idle {type=number}
 FE_Idle_Necklees {type=number}
 FE_Not_Idle_Necklees {type=number}
 FE_NotFEIdle {type=number}
 FrontEnd {type=number}
 Gravity Override {type=Vector}
 HangGlidingAlpha {type=boolean}
 Local Noise Offset {type=Vector}
 Local Noise Transform Location {type=Vector}
 Mat Instance Dynamic {type=MaterialInstanceDynamic}
 No Skydiving {type=number}
 NoDBNO {type=number}
 Overlay Mat Instance Dynamic {type=MaterialInstanceDynamic}
 PlayerPawn {type=FortPlayerPawnAthena}
 R1 Velocity Transform Location {type=Vector}
 R1 Velocity Transform Velocity {type=Vector}
 R2 Velocity Transform Location {type=Vector}
 R2 Velocity Transform Velocity {type=Vector}
 RBAN Settings Out {type=FortRigidBodyAnimNodeParameters}
 Simulation Space {type=AnimPhysSimSpaceType, default=World}
 SkelMesh {type=SkeletalMeshComponent}
 Velocity Interp Speed {type=number}
 WPO Pho Turbulence {type=number}
 WPO Pho Wind Intensity {type=number}
]
